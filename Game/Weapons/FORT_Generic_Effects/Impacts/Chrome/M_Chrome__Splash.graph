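# Masked, two-sided lit material for chrome splash impact effects on Niagara sprites, meshes and ribbons, built on the LackLuster base function.
# A radial flowmap and dynamic CenterWarp distort the splash mask and normal UVs.
# Particle color tints base color, while alpha drives mask erosion; MeshWPOBend scales a vertical-UV Z bend.
# The mobile path skips the warp and LackLuster emissive, deriving normals from the mask.
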Material /Game/Weapons/FORT_Generic_Effects/Impacts/Chrome/M_Chrome__Splash
graph {
 properties {
  bGenerateSphericalParticleNormals=true
  bUsedWithNiagaraMeshParticles=true
  bUsedWithNiagaraRibbons=true
  bUsedWithNiagaraSprites=true
  blendMode=BLEND_Masked
  pixelDepthOffsetMode=PDOM_Legacy
  twoSided=true
  usedShadingModels=Default Lit
 }
 nodes [
  n1 = SphereMask(A=TextureCoordinate(), B=Constant() {r=0.5}, Hardness=Constant() {r=0.25}) {attenuationRadius=1, hardnessPercent=21.6}
  n2 = Subtract(A=TextureCoordinate()) {constB=0.5}
  AppendVector(A=Multiply(A=Multiply(A=@n1, B=@n2)) {constB=5}, B=Constant() {r=1})
  n3 = VectorParameter() {parameterName=Tile, defaultValue=<rgba 1 1 0 0>}
  n4 = TextureCoordinate()
  n5 = TextureSample(UVs=MaterialFunctionCall(Texture Scale=Constant() {r=1.4}) {materialFunction=&ScaleUVsByCenter}.UVs) {
   samplerSource=SSM_Wrap_WorldGroupSettings
   samplerType=SAMPLERTYPE_LinearColor
   texture=/Game/Environments/Hexmap/HexClouds/T_RadialFlowmap01
  }
  n6 = ComponentMask(None=@n5.RGB) {g=true, r=true}
  n7 = DynamicParameter() {parameterIndex=1, defaultValue=<rgba 0 0 1 1>, paramNames [CenterWarp,MeshWPOBend,Param3,Param4]}
  n8 = Add(A=@n4, B=Multiply(A=Subtract(A=@n6) {constB=0.5}, B=@n7.CenterWarp))
  n9 = Multiply(A=AppendVector(A=@n3.R, B=@n3.R), B=ShadingPathSwitch(Default=@n8, Deferred=@n8, Mobile=@n4) {desc=Migrated from FeatureLevelSwitch})
  n10 = TextureSampleParameter2D(UVs=@n9) {
   parameterName=Normal
   samplerSource=SSM_Wrap_WorldGroupSettings
   samplerType=SAMPLERTYPE_Normal
   texture=/Game/Effects/Fort_Effects/Textures/Water/T_water_splash_04_n
  }
  n11 = TextureSampleParameter2D(UVs=@n9) {parameterName=Mask, samplerSource=SSM_Wrap_WorldGroupSettings, texture=/Game/Effects/Fort_Effects/Textures/Water/T_water_splash_04}
  n12 = ChannelMaskParameter(None=@n11.RGB) {maskChannel=Blue, parameterName=ErodeMask, defaultValue=<rgba 0 0 1 0>}
  n13 = Transform(None=MaterialFunctionCall!(Height=@n12.R, Absolute World Position, World Space Vertex Normals) {materialFunction=/Engine/Functions/Engine_MaterialFunctions03/Procedurals/HeightToNormalSmooth}.WorldSpaceNormal) {transformSourceType=TRANSFORMSOURCE_World, transformType=TRANSFORM_Tangent}
  n14 = ScalarParameter() {defaultValue=-25, parameterName=Mobile Flatness}
  LinearInterpolate(
   A=ShadingPathSwitch(
    Default=@n10.RGB
    Mobile=MaterialFunctionCall(Normal=@n13, Flatness=@n14) {materialFunction=&FlattenNormal}.Result
   ) {desc=Migrated from FeatureLevelSwitch}
   B=Constant3Vector() {constant=<rgba 0 0 1 0>}
   Alpha=ScalarParameter() {parameterName=Flatten Normal}
  )
  n15 = BreakMaterialAttributes!(Attr=MaterialFunctionCall!(
    bBlendMaterialAttributes
    MA_In
    UVsOverride
    Override Normals
    BlendMaterialAttributesAlpha
    BaseColorOverride
    WorldPositionOverride
    VertexNormalOverride
    SourceNormalsStrengthOverride
   ) {materialFunction=/Game/TechArt/Seasonal/LackLuster/MaterialFunctions/MF_LackLuster_Base}.MA Out)
  n16 = ShadingPathSwitch(Default=@n15.EmissiveColor, Mobile=Constant()) {desc=Migrated from FeatureLevelSwitch}
  n17 = ParticleColor()
  n18 = ScalarParameter() {parameterName=UseParticleColor}
  MaterialFunctionCall(Emissive=Multiply(A=LinearInterpolate(A=@n16, B=@n17.RGB, Alpha=@n18), B=ScalarParameter() {defaultValue=0.15, parameterName=EmissiveAdd}) {constB=0.15}) {materialFunction=&MF_FX_EmissiveAdjustments}
  n19 = LinearInterpolate(Alpha=@n17.A) {constA=2, constB=-0.1}
  Multiply(A=Multiply(A=@n17.A, B=Power(Base=@n12.R, Exp=@n19)), B=ChannelMaskParameter(None=@n11.RGBA) {parameterName=OpacityMask}.R)
  Multiply(A=@n15.BaseColor, B=@n17.RGB)
  n20 = ComponentMask(None=TextureCoordinate()) {g=true}
  Multiply(A=Multiply(A=Constant3Vector() {constant=<rgba 0 0 1 0>}, B=@n7.MeshWPOBend), B=Power(Base=OneMinus(None=@n20)) {constExponent=3})
 ]
}
